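# Conversation database for the CreamSkull NPC, providing the greeting and top-level branches when a player talks to them.
# The entry point leads to one speech node whose contextual lines cover a first-ever meeting, the PintPiano outfit and per-match conversation count, using priority curve rows where set.
# That node fans out through link nodes to shared quest, basic and random service conversations, a Season39 PerfectGift entry and CreamSkull's own dialogue.
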ConversationDatabase /NPCLibrary/NPCs/CreamSkull/NPCLibrary_Conversation_CreamSkull
entry_points [{tag=Conversation.Entry.CreamSkull, destinations [#1]}]
nodes [
 ConversationEntryPointNode_7 {
  id=#1
  class=ConversationEntryPointNode
  outputs [#2]
  properties {entryTag {tagName=Conversation.Entry.CreamSkull}}
 }
 Speech_C_4 {
  id=#2
  class=Speech_C
  outputs [
   #3
   #4
   #5
   #6
   #7
  ]
  properties {
   nodeName=""
   generalConfig {
    defaultMessage=I'm from the future. I... think?
    contextualMessages [
     {
      message=I'm from the future. I... think?
      contextRequirements [{requirement {class=FortControllerRequirement_IsFirstEverConversationWithNPC}}]
      priority {curve {rowName=Tandem.Speech.Priority.FirstEverInteraction}}
     }
     {
      message="You're the doc, Doc."
      contextRequirements [{requirement {class=FortControllerRequirement_HasCID, properties {allowedCIds [/BRCosmetics/Athena/Items/Cosmetics/Characters/Character_PintPiano]}}}]
      priority {curve {rowName=Tandem.Speech.Priority.OutfitReaction}}
     }
     {
      message=I synchronized my watch and everything.
      contextRequirements [{requirement {class=FortControllerRequirement_CheckNPCMatchConversationCount, properties {nPCUniqueIdTag {tagName=AISpawnerData.Type.Tandem.CreamSkull}}}}]
      priority {curve {rowName=None}}
     }
     {
      message=Nobody calls me chicken.
      contextRequirements [{requirement {class=FortControllerRequirement_CheckNPCMatchConversationCount, properties {nPCUniqueIdTag {tagName=AISpawnerData.Type.Tandem.CreamSkull}}}}]
      priority {curve {rowName=None}}
     }
     {
      message="If you put your mind to it, you can accomplish anything."
      contextRequirements [{requirement {class=FortControllerRequirement_CheckNPCMatchConversationCount, properties {nPCUniqueIdTag {tagName=AISpawnerData.Type.Tandem.CreamSkull}}}}]
      priority {curve {rowName=None}}
     }
    ]
   }
  }
 }
 ConversationLinkNode_3 {
  id=#7
  class=ConversationLinkNode
  properties {remoteEntryTag {tagName=Conversation.Entry.CreamSkull.Dialogue}}
 }
 ConversationLinkNode_2 {
  id=#6
  class=ConversationLinkNode
  properties {remoteEntryTag {tagName=Conversation.Entry.Season39.WF.PerfectGift}}
 }
 ConversationLinkNode_7 {id=#5, class=ConversationLinkNode, properties {remoteEntryTag {tagName=Conversation.Entry.RandomServices}}}
 ConversationLinkNode_0 {id=#4, class=ConversationLinkNode, properties {remoteEntryTag {tagName=Conversation.Entry.BasicServices}}}
 ConversationLinkNode_1 {id=#3, class=ConversationLinkNode, properties {remoteEntryTag {tagName=Conversation.Entry.NPCLibrary.Quests}}}
]
